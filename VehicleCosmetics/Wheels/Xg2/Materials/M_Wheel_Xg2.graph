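# Xg2 vehicle wheel material, layering icy emissive effects on MF_WheelParent_Simple before the shared vehicle damage and de-rez functions.
# Switchable rim lights (masked, desaturated ice texture) and tire lights (flame, cloud and ice textures) blend by vertex red into an Emissive reroute.
# The tire Panners take constant Time, so they give fixed offsets rather than scrolling.
# A tinted cubemap reflection feeds a separate Desaturation reroute.
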Material /VehicleCosmetics/Wheels/Xg2/Materials/M_Wheel_Xg2
graph {
 properties {
  bAutomaticallySetUsageInEditor=false
  bUseMaterialAttributes=true
  bUsedWithMeshDeformer=true
  bUsedWithSkeletalMesh=true
  pixelDepthOffsetMode=PDOM_Legacy
  usedShadingModels=Default Lit
 }
 nodes [
  n1 = GetMaterialAttributes(None=MaterialFunctionCall() {materialFunction=&MF_WheelParent_Simple}.MaterialAttributes) {attributeGetTypes [#1]}
  n2 = LinearInterpolate(A=@n1.Emissive Color, B=NamedRerouteUsage(), Alpha=Constant() {r=1})
  MaterialFunctionCall(In=MaterialFunctionCall(Material=SetMaterialAttributes(MaterialAttributes=@n1.MaterialAttributes, Emissive Color=@n2) {attributeSetTypes [#1]}) {materialFunction=&MF_BaseVehicle_Damage}.Result) {materialFunction=&MF_VehicleDeRez_Transition}
  n3 = TextureSampleParameter2D() {
   group=5_Special Effects
   parameterName=Mask_Emissive
   samplerSource=SSM_Wrap_WorldGroupSettings
   samplerType=SAMPLERTYPE_LinearColor
   texture=/VehicleCosmetics/Wheels/Xg2/Textures/T_Xg2_Wheel_RGB
  }
  n4 = TextureSample(UVs=TextureCoordinate() {uTiling=4, vTiling=4}) {samplerSource=SSM_Wrap_WorldGroupSettings, texture=/VehicleCosmetics/Wheels/RL88/Textures/Ice_01_D}
  n5 = Multiply(
   A=Multiply(A=@n3.R, B=Desaturation(None=@n4.RGBA) {luminanceFactors=<rgba 0.3 0.59 0.11 0>})
   B=ScalarParameter() {defaultValue=1, group=5_Special Effects, parameterName=SnowFlakeBrightness}
  )
  n6 = Constant()
  n7 = ConstantBiasScale(None=OneMinus(None=NamedRerouteUsage())) {bias=-0.5, scale=2}
  n8 = TextureSample(UVs=Panner(Coordinate=TextureCoordinate() {uTiling=0.5, vTiling=0.5}, Time=Constant() {r=0.5}) {speedX=0.1, speedY=0.55}) {samplerSource=SSM_Wrap_WorldGroupSettings, texture=/VehicleCosmetics/Wheels/_SharedTextures/FX/FlameTire_M_T}
  n9 = Multiply(A=Constant4Vector() {constant=<rgba 0.75 0.75 0.75 1>}.A, B=@n8.B)
  n10 = Multiply(
   A=@n8.B
   B=TextureSample(UVs=TextureCoordinate() {uTiling=0.75}) {samplerSource=SSM_Wrap_WorldGroupSettings, texture=&Ice_01_D}.B
  )
  n11 = TextureSample(UVs=Panner(Coordinate=TextureCoordinate() {uTiling=0.5, vTiling=0.5}, Time=Constant() {r=1}) {speedX=0.5, speedY=0.25}) {samplerSource=SSM_Wrap_WorldGroupSettings, texture=/VehicleCosmetics/Wheels/_SharedTextures/FX/Clouds_RGB_T}
  n12 = Multiply(A=Multiply(A=Add(A=@n8.R, B=@n9), B=Add(A=@n10, B=@n11.G)), B=NamedRerouteUsage())
  n13 = NamedRerouteUsage()
  n14 = NamedRerouteUsage()
  n15 = Constant() {r=0.75}
  n16 = Add(A=Multiply(A=Clamp(None=@n7), B=@n12), B=Multiply(A=Multiply(A=@n13, B=@n14), B=Multiply(A=@n12, B=@n15)))
  NamedRerouteDeclaration(None=LinearInterpolate(
    A=StaticSwitchParameter(True=@n5, False=@n6) {defaultValue=true, group=5_Special Effects, parameterName=RimLights}
    B=StaticSwitchParameter(True=@n16, False=@n6) {defaultValue=true, group=5_Special Effects, parameterName=TireLights}
    Alpha=VertexColor().R
   )) {name=Emissive, nodeColor=<rgba 1 0.75 0 1>}
  NamedRerouteDeclaration(None=@n3.B) {name=Rim_Emissive_Blue, nodeColor=<rgba 0 0.6016 1 1>}
  NamedRerouteDeclaration(None=@n3.G) {name=Rim_Emissive_Green, nodeColor=<rgba 0.2372 1 0 1>}
  NamedRerouteDeclaration(None=@n3.A) {name=Rim_Emissive_A, nodeColor=<rgba 1 1 1 1>}
  n17 = TextureSampleParameterCube!(UVs=Transform(None=ReflectionVectorWS()) {transformSourceType=TRANSFORMSOURCE_Local}, Apply View MipBias) {
   group=5_Special Effects
   parameterName=CubeMap
   samplerType=SAMPLERTYPE_LinearColor
   texture=/VehicleCosmetics/Wheels/DuelingDragonsSE/Materials/TC_DuelingDragonsSE_GrassCube
  }
  n18 = Clamp(None=ConstantBiasScale(None=@n17.RGBA) {bias=0.8, scale=0.35})
  n19 = ScalarParameter() {defaultValue=1, group=5_Special Effects, parameterName=ReflectionAmount}
  n20 = VectorParameter() {group=5_Special Effects, parameterName=IceTint, defaultValue=<rgba 0.4322 0.7418 0.9779 4>}
  n21 = Multiply(A=Multiply(A=Multiply(A=@n18, B=@n19), B=@n20.RGBA), B=@n20.A)
  NamedRerouteDeclaration(None=Desaturation(None=Clamp(None=@n21), Fraction=ScalarParameter() {defaultValue=0.75, group=5_Special Effects, parameterName=Desaturation}) {luminanceFactors=<rgba 0.3 0.59 0.11 0>}) {name=Desaturation, nodeColor=<rgba 0 1 0.4375 1>}
 ]
}
